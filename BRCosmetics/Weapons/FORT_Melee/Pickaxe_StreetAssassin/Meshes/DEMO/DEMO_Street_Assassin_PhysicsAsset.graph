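# Physics asset for a demo Street Assassin two-handed pickaxe mesh in Fortnite's cosmetic melee weapons.
# It holds a single capsule body on the Melee_TwoHanded_Handle bone, slightly tilted and offset.
# There are no constraints, so it acts as one rigid collision volume rather than an articulated setup.
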PhysicsAsset /BRCosmetics/Weapons/FORT_Melee/Pickaxe_StreetAssassin/Meshes/DEMO/DEMO_Street_Assassin_PhysicsAsset
bodies {Melee_TwoHanded_Handle [{shapeType=Capsule, length=74.6659, radius=37.191, center=<0.3202 2.8401 31.0857>, rotation=<rot -7.2368 -90 -0.0238>}]}
constraints []
